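# Niagara system for glowing spores on the face accessory of the Shiitake Shaolin Rouge player cosmetic.
# BitsWorld emits drifting curl-noise bits, stopped while driving or skydiving, and SmokeLocal adds soft animated smoke, both from a ring.
# Two burst glow emitters layer two sprites each, with only the HiEnd variant adding a particle light.
# User booleans adjust intensity for front-end preview and drive visibility tags and the light when the head is hidden.
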NiagaraSystem /BRCosmetics/Characters/Player/Male/Medium/Bodies/M_MED_Shiitake_Shaolin/FX/Rouge/NS_Shiitake_Shaolin_Rouge_FaceAcc_Spores
dependencies {
 usedDataInterfaces [&NiagaraDataInterfaceArrayFloat, &NiagaraDataInterfaceCurve, &NiagaraDataInterfaceColorCurve]
 usedDynamicInputs []
 usedModules [&SystemState]
}
emitters {
 BitsWorld {
  emitterUpdateScript [
   &EmitterState()
   &SpawnRate(
    Spawn Probability=&MakeCustomFloatFromBool(Bool=&Boolean_OrOperation(Bool1=$User.IsDriving, Bool2=$User.IsSkydiving), True Float=0, False Float=1)
    SpawnRate=5
    Use Spawn Probability=true
   )
  ]
  particleSpawnScript [
   &InitializeParticle(
    Lifetime=&RandomRangeFloat(Minimum=0.5, Maximum=1.3)
    Position Offset=<0 -10 -5>
    Sprite Rotation Mode=::NewEnumerator2
    Sprite Size Mode=::NewEnumerator1
    Uniform Sprite Size Max=7
    Uniform Sprite Size Min=2
    UsePositionOffset=false
    Write Material Random=true
   )
   &CylinderLocation(
    Cylinder Height=0
    Cylinder Radius=15
    MaskSpawnGroup=false
    Override Seed=false
    Rotation Mode:ENiagara_RotationMode=::NewEnumerator1
    Surface Only=true
    Use Endcaps In Surface Only Mode=false
    UseOwnerScale=true
   )
   &SkeletalMeshLocation(
    MaskSpawnGroup=false
    Mesh Sampling Type=::NewEnumerator5
    Position Sampling=::NewEnumerator1
    Sampled Position Offset=&RandomRangeVector(Minimum=<-3 -3 -3>, Maximum=<3 3 3>)
    Skeletal Mesh=$User.NewNiagaraDataInterfaceSkeletalMesh
    Socket Sampling Mode=::NewEnumerator5
   ) {enabled=false}
   &AddVelocityFromPoint(Module.Use Falloff Distance=false, Velocity Strength=&RandomRangeFloat(Minimum=100, Maximum=200)) {enabled=false}
   &InheritVelocity(Inherited Velocity Amount Scale=<0.5 0.5 0.5>, Speed Limit=true, UseSpeedThreshold=false)
  ]
  particleUpdateScript [
   &ParticleState()
   &InheritSourceMovement(Applied Position Scale=<0.5 0.5 0.5>, ApplyForce=false, ApplyPosition=true, ApplyPositionFalloff=false)
   &CurlNoiseForce(Noise Strength=100, Pan Noise=false)
   &GravityForce(Gravity=&RandomRangeVector(Minimum=<0 0 -25>, Maximum=<0 0 -50>))
   &Drag(Drag=5, Use Linear Drag=true, Use Rotational Drag=false)
   &SolveForcesAndVelocity(Clamp Velocity=false, Limit Acceleration=false)
   &DynamicMaterialParameters(
    Index 0 Param 1=&FloatFromCurve(FloatCurve=DataInterface() {propertyValues {Curve {keys [<at 0 1>,<at 1 0>]}}})
    Index 0 Param 2=&UniformRangedFloat(Minimum=20, Maximum=24)
    Index 0 Param 3=5
    Index 0 Param 4=&FloatFromCurve(FloatCurve=DataInterface() {propertyValues {Curve {keys [<at 0.0123 0.25>,<at 0.99 -0.3>,<at 1 -0.4>]}}})
    Module.Param0WriteEnabled=true
    Module.Param1WriteEnabled=true
    Module.Param2WriteEnabled=true
    Module.Param3WriteEnabled=true
   )
   &ScaleColor(
    Scale Alpha=&FloatFromCurve(
     FloatCurve=DataInterface() {propertyValues {Curve {keys [
         {interpMode=RCIM_Cubic, tangentMode=RCTM_User, time=0, value=1}
         {arriveTangent=-1, interpMode=RCIM_Cubic, leaveTangent=-1, time=0.5, value=0.8}
         {arriveTangent=-0.778, interpMode=RCIM_Cubic, leaveTangent=-0.778, tangentMode=RCTM_User, time=1, value=0}
        ]}}}
     Scale Curve=&MakeCustomFloatFromBool(Bool=$User.bIsFrontEndPreview, True Float=1.5, False Float=1)
    )
    Scale RGB=&MakeVectorFromLinearColorRGB(Color=&ColorFromCurve(ColorCurve=DataInterface() {propertyValues {
        AlphaCurve {keys [<at 0 1 RCIM_Cubic>]}
        BlueCurve {keys [
          {arriveTangent=-2.1985, interpMode=RCIM_Cubic, time=0.0093, value=0.5633}
          {interpMode=RCIM_Cubic, leaveTangent=-2.1985, time=1, value=0.2569}
         ]}
        GreenCurve {keys [
          {arriveTangent=-3.1874, interpMode=RCIM_Cubic, time=0.0093, value=0.9669}
          {interpMode=RCIM_Cubic, leaveTangent=-3.1874, time=1, value=0.4542}
         ]}
        RedCurve {keys [<at 0.0093 2 RCIM_Cubic>,<at 1 0.7396 RCIM_Cubic>]}
       }}))
    ScaleA=true
    ScaleRGB=true
   )
   &ScaleSpriteSize(Uniform Curve Sprite Scale=DataInterface() {propertyValues {Curve {keys [
        {arriveTangent=-1.3333, interpMode=RCIM_Cubic, leaveTangent=-1.3333, time=0, value=1}
        {arriveTangent=-0.4, interpMode=RCIM_Cubic, leaveTangent=-0.4, time=0.3, value=0.6}
        {arriveTangent=-0.5, interpMode=RCIM_Cubic, leaveTangent=-0.5, time=0.5, value=0.8}
        {arriveTangent=-0.5, interpMode=RCIM_Cubic, leaveTangent=-0.5, time=0.7, value=0.4}
        {arriveTangent=-1.3333, interpMode=RCIM_Cubic, leaveTangent=-1.3333, time=0.9, value=0.6}
        {arriveTangent=-6, interpMode=RCIM_Cubic, leaveTangent=-6, time=1, value=0}
       ]}}})
   /SetVariables(Particles.VisibilityTag=&MakeCustomIntFromBool(Bool=$User.bisHeadHidden, True INT=1, False INT=0)) {isSetParameters=true, module=SetVariables_#1}
  ]
  rendererClasses [&NiagaraSpriteRendererProperties]
  renderers [{rendererClass=&NiagaraSpriteRendererProperties}]
 }
 Glow_HiEnd {
  emitterUpdateScript [
   &EmitterState(Life Cycle Mode=::NewEnumerator1, Loop Behavior=::NewEnumerator1, Loop Duration Mode=::NewEnumerator1, UseLoopDelay=false)
   &SpawnBurst_Instantaneous(Use Spawn Probability=false, UseLoopCountLimit=false)
  ]
  particleSpawnScript [
   &InitializeParticle(
    Color=&MakeCustomLinearColorFromBool(Bool=$User.bIsFrontEndPreview, True Color=<rgba 7 3.6455 2.1578 0.2>, False Color=<rgba 9 5.3224 2.8317 0.3>)
    Position Offset=<-4 0 0>
    Sprite Rotation Mode=::NewEnumerator2
    Sprite Size Mode=::NewEnumerator3
    Uniform Sprite Size=30
    UsePositionOffset=true
    Write Material Random=true
   )
   /SetVariables(Particles.Renderer2_Color:LinearColor=&MakeCustomLinearColorFromBool(Bool=$User.bIsFrontEndPreview, True Color=<rgba 0.9911 0.7718 0.5659 1>, False Color=<rgba 15 9.2232 6.7972 0.025>)) {isSetParameters=true, module=SetVariables_#2}
   /SetVariables(Particles.Renderer2_Size:Vector2f=&Multiply_Vector2D_ByFloat(Float=1.4, Vector2D=$Particles.SpriteSize)) {isSetParameters=true, module=SetVariables_#3}
  ]
  particleUpdateScript [
   &ParticleState(Kill Particles When Lifetime Has Elapsed=false, Let Infinitely Lived Particles Die When Emitter Deactivates=true)
   &SolveForcesAndVelocity(Clamp Velocity=false, Limit Acceleration=false)
   &SkeletalMeshLocation(
    MaskSpawnGroup=false
    Mesh Sampling Type=::NewEnumerator5
    Sampled Position Offset=<0 0 3>
    Skeletal Mesh=$User.NewNiagaraDataInterfaceSkeletalMesh
    Socket Sampling Mode=::NewEnumerator5
   ) {enabled=false}
   &DynamicMaterialParameters(
    Index 0 Param 1=&FloatFromCurve(FloatCurve=DataInterface() {propertyValues {Curve {keys [<at 0 1>,<at 1 0>]}}})
    Index 0 Param 2=&UniformRangedFloat(Minimum=20, Maximum=24)
    Index 0 Param 3=4
    Index 0 Param 4=&FloatFromCurve(FloatCurve=DataInterface() {propertyValues {Curve {keys [<at 0.0123 0.25>,<at 0.99 -0.3>,<at 1 -0.4>]}}})
    Module.Param0WriteEnabled=true
    Module.Param1WriteEnabled=true
    Module.Param2WriteEnabled=true
    Module.Param3WriteEnabled=true
   )
   &SpriteRotationRate(Rotation Rate=-250) {enabled=false}
   &CameraOffset(Camera Offset Amount=9)
   &FN_Light_Attributes(
    Color=$Particles.Initial.Renderer2_Color
    Create Color Attribute=true
    Create Exponent Attribute=true
    Create Position Attribute=true
    Create Radius Attribute=true
    Enable Particle Light Enable=true
    Exponent=7
    Light Enabled:NiagaraBool=&Boolean_NotOperation(Bool=$User.bisHeadHidden)
    Radius=50
    Scale Color=0.3
   )
   &FN_Sprite_Attributes(Create Position Attribute=true, Position:Vector3f=<-1 0 0>, Write to Module Namespace=true)
   /SetVariables(Particles.VisibilityTag=&MakeCustomIntFromBool(Bool=$User.bisHeadHidden, True INT=1, False INT=0)) {isSetParameters=true, module=SetVariables_#1}
  ]
  rendererClasses [&NiagaraSpriteRendererProperties, &NiagaraLightRendererProperties]
  renderers [
   {rendererClass=&NiagaraSpriteRendererProperties}
   {rendererClass=&NiagaraSpriteRendererProperties, rendererIndex=1}
   {rendererClass=&NiagaraLightRendererProperties, rendererIndex=2}
  ]
 }
 Glow_LowEand {
  emitterUpdateScript [
   &EmitterState(Life Cycle Mode=::NewEnumerator1, Loop Behavior=::NewEnumerator1, Loop Duration Mode=::NewEnumerator1, UseLoopDelay=false)
   &SpawnBurst_Instantaneous(Use Spawn Probability=false, UseLoopCountLimit=false)
  ]
  particleSpawnScript [
   &InitializeParticle(
    Color=&MakeCustomLinearColorFromBool(Bool=$User.bIsFrontEndPreview, True Color=<rgba 1.2 0.9413 0.5723 0.1>, False Color=<rgba 0.7031 0.5658 0.3325 0.15>)
    Position Offset=<-4 0 0>
    Sprite Rotation Mode=::NewEnumerator2
    Sprite Size Mode=::NewEnumerator3
    Uniform Sprite Size=30
    UsePositionOffset=true
    Write Material Random=true
   )
   /SetVariables(Particles.Renderer2_Color:LinearColor=&MakeCustomLinearColorFromBool(Bool=$User.bIsFrontEndPreview, True Color=<rgba 0 0.9911 0.7952 0.1>, False Color=<rgba 0 15 11.7936 0.0125>)) {isSetParameters=true, module=SetVariables_#2}
   /SetVariables(Particles.Renderer2_Size:Vector2f=&Multiply_Vector2D_ByFloat(Float=1.4, Vector2D=$Particles.SpriteSize)) {isSetParameters=true, module=SetVariables_#3}
  ]
  particleUpdateScript [
   &ParticleState(Kill Particles When Lifetime Has Elapsed=false, Let Infinitely Lived Particles Die When Emitter Deactivates=true)
   &SolveForcesAndVelocity(Clamp Velocity=false, Limit Acceleration=false)
   &SkeletalMeshLocation(
    MaskSpawnGroup=false
    Mesh Sampling Type=::NewEnumerator5
    Sampled Position Offset=<0 0 3>
    Skeletal Mesh=$User.NewNiagaraDataInterfaceSkeletalMesh
    Socket Sampling Mode=::NewEnumerator5
   ) {enabled=false}
   &DynamicMaterialParameters(
    Index 0 Param 1=&FloatFromCurve(FloatCurve=DataInterface() {propertyValues {Curve {keys [<at 0 1>,<at 1 0>]}}})
    Index 0 Param 2=&UniformRangedFloat(Minimum=20, Maximum=24)
    Index 0 Param 3=4
    Index 0 Param 4=&FloatFromCurve(FloatCurve=DataInterface() {propertyValues {Curve {keys [<at 0.0123 0.25>,<at 0.99 -0.3>,<at 1 -0.4>]}}})
    Module.Param0WriteEnabled=true
    Module.Param1WriteEnabled=true
    Module.Param2WriteEnabled=true
    Module.Param3WriteEnabled=true
   )
   &SpriteRotationRate(Rotation Rate=-250) {enabled=false}
   &CameraOffset(Camera Offset Amount=9)
   &FN_Sprite_Attributes(Create Position Attribute=true, Position:Vector3f=<-1 0 0>, Write to Module Namespace=true)
   /SetVariables(Particles.VisibilityTag=&MakeCustomIntFromBool(Bool=$User.bisHeadHidden, True INT=1, False INT=0)) {isSetParameters=true, module=SetVariables_#1}
  ]
  rendererClasses [&NiagaraSpriteRendererProperties]
  renderers [
   {rendererClass=&NiagaraSpriteRendererProperties}
   {rendererClass=&NiagaraSpriteRendererProperties, rendererIndex=1}
  ]
 }
 SmokeLocal {
  emitterUpdateScript [
   &EmitterState()
   &SpawnRate(SpawnRate=&FN_FrontEndFloatModifier(Gameplay=2, Front End=1.2), Use Spawn Probability=false)
  ]
  particleSpawnScript [
   &InitializeParticle(
    Lifetime=&RandomRangeFloat(Minimum=2.5, Maximum=3)
    Position Offset=<0 -10 -5>
    Sprite Rotation Mode=::NewEnumerator2
    Sprite Size=&Multiply_Vector2D_ByFloat(
     Float=&RandomRangeFloat(Minimum=20, Maximum=25)
     Vector2D=&MakeVector2D(
      X=1
      Y=&SelectFloatFromArray(
       Array Sampling Mode=::NewEnumerator0
       Direct Array Index=?
       Float Selection Array=DataInterface() {propertyValues {FloatData [-1,1]}}
       Random Seed=0
      )
     )
    )
    Sprite Size Mode=::NewEnumerator4
    UsePositionOffset=false
    Write Material Random=true
   )
   &AddVelocity(Velocity=&RandomRangeVector(Minimum=<0 25 0>, Maximum=<0 50 0>)) {enabled=false}
   &CylinderLocation(
    Cylinder Height=0
    Cylinder Radius=15
    MaskSpawnGroup=false
    Override Seed=false
    Rotation Mode:ENiagara_RotationMode=::NewEnumerator1
    Surface Only=true
    Use Endcaps In Surface Only Mode=false
    UseOwnerScale=true
   )
  ]
  particleUpdateScript [
   &ParticleState()
   &SkeletalMeshLocation(
    MaskSpawnGroup=false
    Mesh Sampling Type=::NewEnumerator5
    Position Sampling=::NewEnumerator1
    Skeletal Mesh=$User.NewNiagaraDataInterfaceSkeletalMesh
    Socket Sampling Mode=::NewEnumerator5
   ) {enabled=false}
   &AccelerationForce(Acceleration=<0 0 -25>)
   &Drag(Drag=5, Use Linear Drag=true, Use Rotational Drag=false)
   &SolveForcesAndVelocity(Clamp Velocity=false, Limit Acceleration=false)
   &DynamicMaterialParameters(
    Index 0 Param 1=&FloatFromCurve(FloatCurve=DataInterface() {propertyValues {Curve {keys [<at 0 1>,<at 1 0>]}}})
    Index 0 Param 2=&UniformRangedFloat(Minimum=20, Maximum=24)
    Index 0 Param 3=5
    Index 0 Param 4=&FloatFromCurve(FloatCurve=DataInterface() {propertyValues {Curve {keys [<at 0.0123 0.25>,<at 0.99 -0.3>,<at 1 -0.4>]}}})
    Module.Param0WriteEnabled=true
    Module.Param1WriteEnabled=true
    Module.Param2WriteEnabled=true
    Module.Param3WriteEnabled=true
   )
   &SpriteRotationRate(Rotation Rate=-250) {enabled=false}
   &SubUVAnimation(Start Frame=38)
   &ScaleColor(
    Scale Alpha=&FloatFromCurve(
     FloatCurve=DataInterface() {propertyValues {Curve {keys [
         {interpMode=RCIM_Cubic, tangentMode=RCTM_User, time=0, value=1}
         {arriveTangent=-1, interpMode=RCIM_Cubic, leaveTangent=-1, time=0.5, value=0.8}
         {arriveTangent=-0.778, interpMode=RCIM_Cubic, leaveTangent=-0.778, tangentMode=RCTM_User, time=1, value=0}
        ]}}}
     Scale Curve=&MakeCustomFloatFromBool(Bool=$User.bIsFrontEndPreview, True Float=0.5, False Float=0.1)
    )
    Scale RGB=&MakeVectorFromLinearColorRGB(Color=&ColorFromCurve(ColorCurve=DataInterface() {propertyValues {
        AlphaCurve {keys [<at 0 1 RCIM_Cubic>]}
        BlueCurve {keys [{arriveTangent=-2.1985, interpMode=RCIM_Cubic, time=0.0093, value=0.574}, {interpMode=RCIM_Cubic, leaveTangent=-2.1985, time=1, value=0.284}]}
        GreenCurve {keys [
          {arriveTangent=-3.1874, interpMode=RCIM_Cubic, time=0.0093, value=0.8545}
          {interpMode=RCIM_Cubic, leaveTangent=-3.1874, time=1, value=0.4263}
         ]}
        RedCurve {keys [<at 0.0093 2 RCIM_Cubic>,<at 1 0.7396 RCIM_Cubic>]}
       }}))
    ScaleA=true
    ScaleRGB=true
   )
   &ScaleSpriteSize(Uniform Curve Sprite Scale=DataInterface() {propertyValues {Curve {keys [
        {arriveTangent=2.1622, interpMode=RCIM_Cubic, leaveTangent=2.1622, tangentMode=RCTM_User, time=0, value=0.2}
        {interpMode=RCIM_Cubic, tangentMode=RCTM_User, time=0.5, value=1}
        {arriveTangent=-2.478, interpMode=RCIM_Cubic, leaveTangent=-2.478, tangentMode=RCTM_User, time=1, value=0.4}
       ]}}})
   /SetVariables(Particles.VisibilityTag=&MakeCustomIntFromBool(Bool=$User.bisHeadHidden, True INT=1, False INT=0)) {isSetParameters=true, module=SetVariables_#1}
  ]
  rendererClasses [&NiagaraSpriteRendererProperties]
  renderers [{rendererClass=&NiagaraSpriteRendererProperties}]
 }
}
summary {
 systemName=NS_Shiitake_Shaolin_Rouge_FaceAcc_Spores
 emitters [
  {emitterName=SmokeLocal, rendererClasses [&NiagaraSpriteRendererProperties]}
  {emitterName=Glow_HiEnd, rendererClasses [&NiagaraSpriteRendererProperties, &NiagaraLightRendererProperties]}
  {emitterName=BitsWorld, rendererClasses [&NiagaraSpriteRendererProperties]}
  {emitterName=Glow_LowEand, rendererClasses [&NiagaraSpriteRendererProperties]}
 ]
 userVariables [
  User.bIsFrontEnd {type=NiagaraBool}
  User.bIsFrontEndPreview {type=NiagaraBool}
  User.bisHeadHidden {type=NiagaraBool}
  User.IsDriving {type=NiagaraBool}
  User.IsSkydiving {type=NiagaraBool}
  User.NewNiagaraDataInterfaceSkeletalMesh {
   type=NiagaraDataInterfaceSkeletalMesh
   defaultValue {
    struct=NiagaraExt_VariableValue_DataInterface
    value {dataInterface=&self:NiagaraDataInterfaceSkeletalMesh_0}
   }
  }
 ]
}
